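# Physics asset for the Neon Cat Spy backpack cosmetic, defining collision bodies and joints for its dangling accessories.
# A box-and-capsule pack body anchors a box bag handle, a box sword and a two-link capsule grenade chain.
# The grenades swing widely, the second hanging from the first, while the handle has tighter limits.
# The sword locks both swing axes, leaving only limited twist about its joint.
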PhysicsAsset /BRCosmetics/Accessories/FORT_Backpacks/Backpack_F_MED_Neon_Cat_Spy/Meshes/F_MED_Neon_Cat_Spy_Physics
bodies {
 dyn_bagHandle [{shapeType=Box, extentX=7, extentY=7.5, extentZ=2, center=<1 0 0>}]
 dyn_grenade_01 [{shapeType=Capsule, length=1, radius=1, center=<1 0 0>, rotation=<rot 90 0 0>}]
 dyn_grenade_02 [{shapeType=Capsule, length=1.2958, radius=3.4249, center=<3.913 -0.4994 0.3149>, rotation=<rot 0 -67.5976 -58.9884>}]
 dyn_sword [{shapeType=Box, extentX=30, extentY=5, extentZ=2, center=<0 -2 0>}]
 pack [
  {shapeType=Capsule, length=7.5476, radius=5.3444, center=<-4.2248 -3.4692 -8.3647>, rotation=<rot 10 0 0>}
  {shapeType=Box, extentX=13.1935, extentY=7.5, extentZ=27.6391, center=<1.3713 -5.3504 -0.7848>, rotation=<rot -5 0 0>}
 ]
}
constraints [
 {
  bone1Name=dyn_sword
  bone2Name=pack
  swing1LimitDegrees=45
  swing1Motion=Locked
  swing2LimitDegrees=45
  swing2Motion=Locked
  twistLimitDegrees=60
 }
 {bone1Name=dyn_grenade_01, bone2Name=pack, swing1LimitDegrees=90, swing2LimitDegrees=90, twistLimitDegrees=30}
 {bone1Name=dyn_grenade_02, bone2Name=dyn_grenade_01, swing1LimitDegrees=90, swing2LimitDegrees=90, twistLimitDegrees=30}
 {bone1Name=dyn_bagHandle, bone2Name=pack, swing1LimitDegrees=15, swing2LimitDegrees=45, twistLimitDegrees=5}
]
